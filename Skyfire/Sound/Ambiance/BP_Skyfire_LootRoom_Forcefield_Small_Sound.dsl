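; Ambient sound Blueprint for a small loot room forcefield in Skyfire, built on a box-shaped audio shape.
; On BeginPlay it reads the Box component's world transform, converts it, applies it to the FortAudioShapeBox and enables that shape.
; The conversion keeps location and rotation but multiplies each scale axis by its own AudioShape X, Y or Z variable, so the audio volume derives from the placed box.
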(event EventBeginPlay
  ;[0 0.7101 1] Initialize
  (bind _returnvalue (CallFunction|ConvertTransformValuesforAudioShape (Transformation|GetWorldTransform (Variables|Default|GetBox))))
  (Control|SetBoxTransform (Variables|Default|GetFortAudioShapeBox) _returnvalue)
  (Shape|Enable (Variables|Default|GetFortAudioShapeBox)))

(fn ConvertTransformValuesForAudioShape
  (params (InTransform :type Transform%20%28by%20ref%29))
  (return (Math|Transform|MakeTransform (.location InTransform) (.rotation InTransform) (Math|Vector|MakeVector (* (.x (.scale InTransform)) (Variables|Default|GetAudioShape_X_Value)) (* (.y (.scale InTransform)) (Variables|Default|GetAudioShape_Y_Value)) (* (.z (.scale InTransform)) (Variables|Default|GetAudioShape_Z_Value))))))

(fn ConstructionScript)
